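; Season 29 Battle Pass quest ability that tracks whether the player damages the same valid target with both Mythic and non-Mythic sources.
; On activation it listens for Event.Damage.Dealt on the avatar and filters by ValidTargets tags.
; A map keyed by target player name stores a Vector2D tally: X counts Mythic hits, Y counts all others.
; Repeat hits broadcast the quest's custom verb whenever both tallies are above zero.
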(event Ability|EventActivateAbility
  (bind _map (Variables|Default|GetMap))
  (bind _map_2 (Variables|Default|GetMap))
  (bind _map_3 (Variables|Default|GetMap))
  (|CallParentFunction)
  (bind (_asynctaskproxy _payload) (Ability|Async|WaitGameplayEventToActor (Ability|GetAvatarActorfromActorInfo) "(TagName=\"Event.Damage.Dealt\")" false false)
    (:EventReceived
      (if (GameplayTags|HasAnyTags (.TargetTags (Utilities|Struct|BreakGameplayEventData _payload)) (Variables|Default|GetValidTargets))
        (if (.ReturnValue (Utilities|Map|Find _map_2 (Fortnite|GetPlayerNameSafe (.Target (Utilities|Struct|BreakGameplayEventData _payload)))))
          (if (GameplayTags|HasTag (.InstigatorTags (Utilities|Struct|BreakGameplayEventData _payload)) "(TagName=\"Rarity.Mythic\")")
            (Utilities|Map|Add _map (Fortnite|GetPlayerNameSafe (.Target (Utilities|Struct|BreakGameplayEventData _payload))) (+ (Utilities|Map|Find _map_2 (Fortnite|GetPlayerNameSafe (.Target (Utilities|Struct|BreakGameplayEventData _payload)))) "(X=1.000000,Y=0.000000)"))
            (else
              (Utilities|Map|Add _map (Fortnite|GetPlayerNameSafe (.Target (Utilities|Struct|BreakGameplayEventData _payload))) (+ (Utilities|Map|Find _map_2 (Fortnite|GetPlayerNameSafe (.Target (Utilities|Struct|BreakGameplayEventData _payload)))) "(X=0.000000,Y=1.000000)"))))
          (if (and (> (Math|Vector2D|BreakVector2D (Utilities|Map|Find _map_2 (Fortnite|GetPlayerNameSafe (.Target (Utilities|Struct|BreakGameplayEventData _payload))))) 0) (> (.Y (Math|Vector2D|BreakVector2D (Utilities|Map|Find _map_2 (Fortnite|GetPlayerNameSafe (.Target (Utilities|Struct|BreakGameplayEventData _payload)))))) 0))
            (CustomVerb|BroadcastCustomVerbMessage (CustomVerb|MakeCustomVerbMessage (Ability|GetAvatarActorfromActorInfo) (Variables|Default|GetVerb%20Name))))
          (elif (GameplayTags|HasTag (.InstigatorTags (Utilities|Struct|BreakGameplayEventData _payload)) "(TagName=\"Rarity.Mythic\")")
            (Utilities|Map|Add _map_3 (Fortnite|GetPlayerNameSafe (.Target (Utilities|Struct|BreakGameplayEventData _payload))) "(X=1.000000,Y=0.000000)")
            (else
              (Utilities|Map|Add _map_3 (Fortnite|GetPlayerNameSafe (.Target (Utilities|Struct|BreakGameplayEventData _payload))) "(X=0.000000,Y=1.000000)"))))))))

(event Ability|EventOnEndAbility
  (params (bWasCancelled :type Boolean)))
